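; BPI_BeamAudio_Interface is a Blueprint interface for controlling beam weapon audio.
; It declares calls to play or stop beam audio, set named float parameters, and assign or query the owning Fort weapon.
; An update call passes beam start and end points, the hit actor, hit success and a default-only flag.
; A getter exposes the beam audio actor; the file holds only signatures, with no logic.
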(fn GetBeamAudioActor
  (return 0))

(fn GetBeamOwner
  (return 0 false))

(fn PlayBeamAudio
  (params (bPlay :type Boolean)))

(fn SetBeamAudioFloatParam
  (params (Name :type Name) (Float :type Float%20%28double-precision%29)))

(fn SetBeamOwner
  (params (WeaponOwner :type Fort%20Weapon%20Object%20Reference)))

(fn UpdateBeamAudio
  (params (StartPoint :type Vector) (EndPoint :type Vector) (HitActor :type Actor%20Object%20Reference) (bUseDefaultOnly :type Boolean) (bHitSuccess :type Boolean)))
