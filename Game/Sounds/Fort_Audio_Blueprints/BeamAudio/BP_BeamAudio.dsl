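; Beam weapon audio actor for Fortnite, driven through BeamAudio events and configured from a BP_BeamAudioData asset.
; Play spawns 1p/3p start sounds, triggers weapon fire tags and shows non-local sound indicators; stop plays the stop tag and fades out.
; Updates stretch an AudioShape line along the beam, which can be removed in first person, and move the actor to the endpoint.
; Hit actors classify as water, pawn, generic or default, driving a lazily created surface sound component.
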(event EventBeginPlay
  (bind _returnvalue (Actor|GetOwner))
  (Helper|InitializeConfig)
  (Utilities|IsValid _returnvalue
    (:Is%20Valid
      (bind _asfort_weapon (Utilities|Casting|CastToFortWeapon _returnvalue))
      (BeamAudio|SetBeamOwner _asfort_weapon :!class /Game/Sounds/Fort_Audio_Blueprints/BeamAudio/BPI_BeamAudio_Interface.BPI_BeamAudio_Interface_C)
      (CallFunction|UsePerspectiveAudioShape))))

(event BeamAudio|EventPlayBeamAudio
  (params (bPlay :type Boolean))
  (if bPlay
    (CallFunction|PlayBeamFire)
    (bind _audioshapeline (Variables|BeamAudioConfig|GetAudioShapeLine)
      (:then
        (if (not (Components|Activation|IsActive _audioshapeline))
          (Utilities|IsValid (Variables|/Script/AudioShapes.AudioShapePrimitiveComponent|GetSoundOnEdge _audioshapeline)
            (:Is%20Valid
              (Components|Activation|Activate _audioshapeline false)))))
      (:else
        _))
    (Variables|Private|SetbIsPlaying true)
    (else
      (bind _fortweaponref (Variables|Transient|GetFortWeaponRef)
        (:then
          (SoundLibrary|PlaySoundSimple (WorldSubsystems|GetSoundLibrarySubsystem) _fortweaponref (Utilities|Struct|MakeSimpleSoundSettings (Variables|/Game/Sounds/Fort_Audio_Blueprints/BeamAudio/BP_BeamAudioData.BP_BeamAudioData_C|GetWeaponFireOnStopTag (Variables|BeamAudioConfig|GetBeamAudioConfig)) true false true false 1.0 1.0 0.0 0 0 (Effects|GetMuzzleSocketName (Variables|Transient|GetFortWeaponRef) 0)) :!class /Script/SoundLibrary.SoundLibrarySubsystem)
          (Utilities|FlowControl|Sequence
            (:then_0
              (bind _beamstartaudiocomponent (Variables|Default|GetBeamStartAudioComponent)
                (:then
                  (Audio|Components|Audio|FadeOut _beamstartaudiocomponent 0.1 0.0 "Linear" :!class /Script/Engine.AudioComponent))
                (:else)))
            (:then_1
              (bind _audioshapeline_2 (Variables|BeamAudioConfig|GetAudioShapeLine)
                (:then
                  (Components|Activation|Deactivate _audioshapeline_2))
                (:else)))
            (:then_2
              (Variables|Transient|SetPreviousMat "SurfaceType_Default")
              (Variables|Private|SetbIsPlaying false))
            (:then_3
              (bind _surface_sound_component (Variables|Default|GetSurface%20Sound%20Component)
                (:then
                  (Components|DestroyComponent _surface_sound_component))
                (:else)))))
        (:else)))))

(event EventDestroyed
  (Utilities|FlowControl|Sequence
    (:then_0
      (bind _beamstartaudiocomponent (Variables|Default|GetBeamStartAudioComponent)
        (:then
          (Audio|Components|Audio|FadeOut _beamstartaudiocomponent 0.1 0.0 "Linear" :!class /Script/Engine.AudioComponent))
        (:else)))
    (:then_1
      (bind _audioshapeline (Variables|BeamAudioConfig|GetAudioShapeLine)
        (:then
          (Components|Activation|Deactivate _audioshapeline))
        (:else)))
    (:then_2
      (Variables|Transient|SetPreviousMat "SurfaceType_Default")
      (Variables|Private|SetbIsPlaying false))
    (:then_3
      (bind _surface_sound_component (Variables|Default|GetSurface%20Sound%20Component)
        (:then
          (Components|DestroyComponent _surface_sound_component))
        (:else)))))

(event BeamAudio|EventUpdateBeamAudio
  (params (StartPoint :type Vector :default "0, 0, 0") (EndPoint :type Vector :default "0, 0, 0") (HitActor :type Actor%20Object%20Reference) (bUseDefaultOnly :type Boolean) (bHitSuccess :type Boolean))
  (Variables|Private|GetbIsPlaying
    (:then
      (CallFunction|UpdateBeamLocation StartPoint EndPoint)
      (bind (_change _surface) (Helper|ActortoSurface HitActor bUseDefaultOnly bHitSuccess))
      (if _change
        (bind _output_get (Variables|Transient|SetPreviousMat _surface))
        (bind _surfacesoundcomponent (CallFunction|AddSurfaceSoundComponent))
        (Class|SurfaceSoundComponent|SurfaceTypeChanged _surfacesoundcomponent _output_get)))
    (:else)))

(event BeamAudio|EventSetBeamAudioFloatParam
  (params (Name :type Name :default "None") (Float :type Float%20%28double-precision%29))
  (bind _audioshapeline (Variables|BeamAudioConfig|GetAudioShapeLine)
    (:then
      (Utilities|IsValid (Control|GetEdgeAudioComponent _audioshapeline)
        (:Is%20Valid
          (Audio|Parameter|SetFloatParameter (Control|GetEdgeAudioComponent _audioshapeline) Name Float :!class /Script/Engine.AudioComponent))))
    (:else)))

(event BeamAudio|EventSetBeamOwner
  (params (WeaponOwner :type Fort%20Weapon%20Object%20Reference))
  (bind _output_get (Variables|Transient|SetFortWeaponRef WeaponOwner))
  (Transformation|SetActorLocation (Effects|GetMuzzleLocation _output_get -1)))

(fn ActorToSurface
  (params (HitActor :type Actor%20Object%20Reference) (bUseDefaultOnly :type Boolean) (bHitSuccess :type Boolean))
  (bind _hitactor (Variables|GetHitActor))
  (if bUseDefaultOnly
    ; Default only logic
    (Variables|GetbHitSuccess
      (:then
        (Utilities|FlowControl|Sequence
          (:then_0
            (LocalVariables|SetNewSurface "SurfaceType10"))
          (:then_1
            (return (Utilities|Enum|NotEqual(Enum) (Variables|Transient|GetPreviousMat) (LocalVariables|GetNewSurface)) (LocalVariables|GetNewSurface)))))
      (:else
        (return true "SurfaceType_Default")))
    (else
      (Variables|GetHitActor
        (:then
          (Utilities|FlowControl|Sequence
            (:then_0
              (Utilities|IsA(soft) _hitactor "/Script/Water.WaterBody"
                (:True
                  (LocalVariables|SetNewSurface "SurfaceType8"))
                (:False
                  (Utilities|IsA(soft) _hitactor "/Script/FortniteGame.FortPawn"
                    (:True
                      (LocalVariables|SetNewSurface "SurfaceType4"))
                    (:False
                      (Variables|GetbHitSuccess
                        (:then
                          (LocalVariables|SetNewSurface "SurfaceType10"))
                        (:else
                          (LocalVariables|SetNewSurface "SurfaceType_Default")
                          (return true "SurfaceType_Default"))))))))
            (:then_1
              (return (Utilities|Enum|NotEqual(Enum) (Variables|Transient|GetPreviousMat) (LocalVariables|GetNewSurface)) (LocalVariables|GetNewSurface)))))
        (:else
          (return true "SurfaceType_Default"))))))

(fn AddSurfaceSoundComponent
  (bind _surface_sound_component (Variables|Default|GetSurface%20Sound%20Component)
    (:then
      (return _surface_sound_component))
    (:else
      (bind _returnvalue (Game|AddComponentbyClass 0 "/Game/Sounds/Fort_Audio_Blueprints/SurfaceTypeSound/BP_SurfaceSoundComponent.BP_SurfaceSoundComponent_C" (Variables|Default|GetDefaultSceneRoot) (Variables|Private|GetInternalBeamImpactSound) (Variables|Private|GetInternalSurfaceSoundDataTable)))
      (bind _output_get (Variables|Default|SetSurface%20Sound%20Component _returnvalue))
      (return _output_get))))

(fn GetBeamAudioActor
  (return self))

(fn GetBeamOwner
  (bind _fortweaponref (Variables|Transient|GetFortWeaponRef)
    (:then
      (return _fortweaponref true))
    (:else
      (return 0 false))))

(fn GetWeaponSoundLibraryComponent
  (bind _fortweaponref (Variables|Transient|GetFortWeaponRef)
    (:then
      (return (SoundLibrary|GetSoundLibraryComponent _fortweaponref "(TagName=\"SoundLibrary.Weapon\")" :!class /Script/FortniteGame.FortWeapon)))
    (:else
      (return 0))))

(fn InitializeConfig
  (bind _beamaudioconfig (Variables|BeamAudioConfig|GetBeamAudioConfig)
    (:then
      (Variables|Private|SetInternalSurfaceSoundDataTable (Variables|/Game/Sounds/Fort_Audio_Blueprints/BeamAudio/BP_BeamAudioData.BP_BeamAudioData_C|GetSurfaceSoundDataTable _beamaudioconfig))
      (Variables|Private|SetInternalBeamLoopSound (Variables|/Game/Sounds/Fort_Audio_Blueprints/BeamAudio/BP_BeamAudioData.BP_BeamAudioData_C|GetBeamLoopSound _beamaudioconfig))
      (Variables|Private|SetInternalBeamImpactSound (Variables|/Game/Sounds/Fort_Audio_Blueprints/BeamAudio/BP_BeamAudioData.BP_BeamAudioData_C|GetBeamImpactSound _beamaudioconfig))
      (Variables|Private|SetInternalBeamStart1pSound (Variables|/Game/Sounds/Fort_Audio_Blueprints/BeamAudio/BP_BeamAudioData.BP_BeamAudioData_C|GetBeamStartPoint1pSound _beamaudioconfig))
      (Variables|Private|SetInternalBeamStart3pSound (Variables|/Game/Sounds/Fort_Audio_Blueprints/BeamAudio/BP_BeamAudioData.BP_BeamAudioData_C|GetBeamStartPoint3pSound _beamaudioconfig))
      (Variables|Private|SetbInternalDisable1PSpline (Variables|/Game/Sounds/Fort_Audio_Blueprints/BeamAudio/BP_BeamAudioData.BP_BeamAudioData_C|GetbDisable1pBeamLoopSound _beamaudioconfig))
      (bind _audioshapeline (Variables|BeamAudioConfig|GetAudioShapeLine)
        (:then
          (Variables|/Script/AudioShapes.AudioShapeComponent|SetFadeInTime (Variables|/Game/Sounds/Fort_Audio_Blueprints/BeamAudio/BP_BeamAudioData.BP_BeamAudioData_C|GetBeamLoopFadeInTime _beamaudioconfig) _audioshapeline)
          (Variables|/Script/AudioShapes.AudioShapeComponent|SetFadeOutTime (Variables|/Game/Sounds/Fort_Audio_Blueprints/BeamAudio/BP_BeamAudioData.BP_BeamAudioData_C|GetBeamLoopFadeOutTime _beamaudioconfig) _audioshapeline))
        (:else)))
    (:else)))

(fn PlayBeamFire
  (bind _returnvalue (Helper|GetWeaponSoundLibraryComponent))
  (bind _fortweaponref (Variables|Transient|GetFortWeaponRef)
    (:then
      (CallFunction|PlayStartPointSound _fortweaponref)
      (bind _instigator (Variables|/Script/Engine.Actor|GetInstigator _fortweaponref)
        (:then
          (Utilities|FlowControl|Sequence
            (:then_0
              (Default|SpawnSoundIndicator))
            (:then_1
              (Utilities|IsValid _returnvalue
                (:Is%20Valid
                  (bind _returnvalue_2 (SoundLibrary|PlayEvent _returnvalue (select (Pawn|IsLocallyViewed _instigator) (Variables|/Game/Sounds/Fort_Audio_Blueprints/BeamAudio/BP_BeamAudioData.BP_BeamAudioData_C|GetWeaponFire1pTag (Variables|BeamAudioConfig|GetBeamAudioConfig)) (Variables|/Game/Sounds/Fort_Audio_Blueprints/BeamAudio/BP_BeamAudioData.BP_BeamAudioData_C|GetWeaponFire3pTag (Variables|BeamAudioConfig|GetBeamAudioConfig)))))
                  (if (Utilities|Array|IsValidIndex (.AudioComponents (Utilities|Struct|BreakSoundLibraryPlayResult _returnvalue_2)))
                    (Audio|Parameter|ExecuteTriggerParameter (Utilities|Array|Get(acopy) (.AudioComponents (Utilities|Struct|BreakSoundLibraryPlayResult _returnvalue_2))) "FireSingleShot")
                    (return)))))))
        (:else)))
    (:else)))

(fn PlayStartPointSound
  (params (Weapon :type Actor%20Object%20Reference))
  (bind _rootcomponent (Variables|/Script/Engine.Actor|GetRootComponent Weapon))
  (Variables|Private|GetbIsLocal
    (:then
      (bind _internalbeamstart1psound (Variables|Private|GetInternalBeamStart1pSound)
        (:then
          (bind _returnvalue (Audio|SpawnSoundAttached _internalbeamstart1psound _rootcomponent))
          (Variables|Default|SetBeamStartAudioComponent _returnvalue))
        (:else)))
    (:else
      (bind _internalbeamstart3psound (Variables|Private|GetInternalBeamStart3pSound)
        (:then
          (bind _returnvalue_2 (Audio|SpawnSoundAttached _internalbeamstart3psound _rootcomponent))
          (Variables|Default|SetBeamStartAudioComponent _returnvalue_2))
        (:else)))))

(fn SpawnSoundIndicator
  ;[0.02051 0.132 0.15 0.5] Spawn the sound indicator
  (bind _internalbeamloopsound (Variables|Private|GetInternalBeamLoopSound)
    (:then
      (if (not (Variables|Private|GetbIsLocal))
        (Fortnite|ShowSoundIndicatorLocalforTeam :Location (Transformation|GetActorLocation) :MaxAudibleDistance (Variables|/Script/Engine.SoundBase|GetMaxDistance _internalbeamloopsound) :IndicatorType "Gunshot" :Team (Team|GetActorTeam (Variables|/Script/Engine.Actor|GetInstigator (Variables|Transient|GetFortWeaponRef))) :Affiliations (Utilities|Array|MakeArray "Neutral" "Hostile" "Friendly") :Tint "(R=0.000000,G=0.000000,B=0.000000,A=0.000000)" :SoundBase (Variables|Private|GetInternalBeamLoopSound) :AttenuationMaxRangeMultiplier 1.0 :AttenuationIntensityMultiplier 1.0)))
    (:else)))

(fn UpdateBeamLocation
  (params (StartPoint :type Vector) (Endpoint :type Vector))
  (Utilities|FlowControl|Sequence
    (:then_0
      (bind _audioshapeline (Variables|BeamAudioConfig|GetAudioShapeLine)
        (:then
          (if (not (Components|Activation|IsActive _audioshapeline))
            (Components|Activation|Activate (Variables|BeamAudioConfig|GetAudioShapeLine) false))
          (Control|SetEndPoint (Variables|BeamAudioConfig|GetAudioShapeLine) (Variables|GetEndpoint))
          (Control|SetStartPoint (Variables|BeamAudioConfig|GetAudioShapeLine) (Variables|GetStartPoint)))
        (:else)))
    (:then_1
      (Transformation|SetActorLocation (Variables|GetEndpoint) true))))

(fn UsePerspectiveAudioShape
  (bind _fortweaponref (Variables|Transient|GetFortWeaponRef)
    (:then
      (bind _instigator (Variables|/Script/Engine.Actor|GetInstigator _fortweaponref)
        (:then
          (Variables|Private|SetbIsLocal (Pawn|IsLocallyViewed _instigator))
          (if (and (Variables|Private|GetbIsLocal) (Variables|Private|GetbInternalDisable1PSpline))
            (Components|DestroyComponent (Variables|BeamAudioConfig|GetAudioShapeLine))))
        (:else)))
    (:else)))

(fn ConstructionScript)
